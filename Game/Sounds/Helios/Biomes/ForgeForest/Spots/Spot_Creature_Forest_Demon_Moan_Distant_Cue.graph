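# Sound cue for a distant forest demon moan, a creature ambience one-shot for the ForgeForest biome.
# A quality-level node feeds its first two levels through a shared modulator into an equal-weight random pick of eleven moan waves, leaving the other two levels empty.
# It plays at low volume with a distant-creature attenuation asset, an exterior ambient bus send, and distance-ranged sends to distant and default reverb submixes.
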SoundCue /Game/Sounds/Helios/Biomes/ForgeForest/Spots/Spot_Creature_Forest_Demon_Moan_Distant_Cue
root=QualityLevel_0
nodes [
 {id=Modulator_0, children [Random_0]}
 {id=QualityLevel_0, children [Modulator_0,Modulator_0,null,null]}
 {
  id=Random_0
  children [
   WavePlayer_0
   WavePlayer_1
   WavePlayer_2
   WavePlayer_3
   WavePlayer_4
   WavePlayer_5
   WavePlayer_6
   WavePlayer_11
   WavePlayer_12
   WavePlayer_13
   WavePlayer_15
  ]
  properties {weights [1,1,1,1,1,1,1,1,1,1,1]}
 }
 {id=WavePlayer_0, properties {soundWaveAssetPtr=&/Spot_Creature_Forest_Demon_Moan_Distant_01}}
 {id=WavePlayer_1, properties {soundWaveAssetPtr=&/Spot_Creature_Forest_Demon_Moan_Distant_02}}
 {id=WavePlayer_11, properties {soundWaveAssetPtr=&/Spot_Creature_Forest_Demon_Moan_Distant_08}}
 {id=WavePlayer_12, properties {soundWaveAssetPtr=&/Spot_Creature_Forest_Demon_Moan_Distant_09}}
 {id=WavePlayer_13, properties {soundWaveAssetPtr=&/Spot_Creature_Forest_Demon_Moan_Distant_10}}
 {id=WavePlayer_15, properties {soundWaveAssetPtr=&/Spot_Creature_Forest_Demon_Moan_Distant_12}}
 {id=WavePlayer_2, properties {soundWaveAssetPtr=&/Spot_Creature_Forest_Demon_Moan_Distant_03}}
 {id=WavePlayer_3, properties {soundWaveAssetPtr=&/Spot_Creature_Forest_Demon_Moan_Distant_04}}
 {id=WavePlayer_4, properties {soundWaveAssetPtr=&/Spot_Creature_Forest_Demon_Moan_Distant_05}}
 {id=WavePlayer_5, properties {soundWaveAssetPtr=&/Spot_Creature_Forest_Demon_Moan_Distant_06}}
 {id=WavePlayer_6, properties {soundWaveAssetPtr=&/Spot_Creature_Forest_Demon_Moan_Distant_07}}
]
properties {
 attenuationSettings=/Game/Sounds/Attenuation/Fort_TileAmbient_OneShot_Distant_Creature_Att
 maxDistance=8500
 soundClassObject=&Fort_Ambient_OneShots_Outdoors
 volumeMultiplier=0.1
 attenuationOverrides {bAttenuate=false}
 busSends [{audioBus=null, minSendLevel=1, sendLevel=1, soundSourceBus=&AmbientAudio_Exterior_SourceBus}]
 soundSubmixSends [
  {
   maxSendDistance=8000
   minSendDistance=4000
   minSendLevel=0
   soundSubmix=&Fortnite_Reverb_Distant
   customSendLevelCurve {editorCurveData {keys [<at 0 0.33>,<at 1 1>]}}
  }
  {
   maxSendDistance=8000
   minSendDistance=4000
   minSendLevel=0.25
   soundSubmix=&Fortnite_Reverb_Default
   customSendLevelCurve {externalCurve=/Game/Sounds/SubMixes/Reverb/Fortnite_Reverb_Distant_Send}
  }
 ]
}
