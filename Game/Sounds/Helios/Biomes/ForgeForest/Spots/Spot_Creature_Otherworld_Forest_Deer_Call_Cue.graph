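# Sound cue for distant otherworldly deer calls, an outdoor ambient one-shot in the ForgeForest biome.
# A quality-level node sends its first two slots to a random pick of ten call waves and its last two to a smaller pool of four.
# Each pool has its own modulator for pitch and volume variation.
# Playback is quiet, using a distant-creature attenuation asset and the outdoor one-shot sound class.
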SoundCue /Game/Sounds/Helios/Biomes/ForgeForest/Spots/Spot_Creature_Otherworld_Forest_Deer_Call_Cue
root=QualityLevel_0
nodes [
 {id=Modulator_1, children [Random_3], properties {pitchMax=1.1, pitchMin=0.7, volumeMax=1.1, volumeMin=0.9}}
 {id=Modulator_2, children [Random_4], properties {pitchMax=1, pitchMin=0.8, volumeMax=1.1, volumeMin=0.9}}
 {id=QualityLevel_0, children [Modulator_1,Modulator_1,Modulator_2,Modulator_2]}
 {
  id=Random_3
  children [
   WavePlayer_34
   WavePlayer_35
   WavePlayer_36
   WavePlayer_37
   WavePlayer_38
   WavePlayer_39
   WavePlayer_40
   WavePlayer_41
   WavePlayer_42
   WavePlayer_43
  ]
  properties {weights [1,1,1,1,1,1,1,1,1,1]}
 }
 {id=Random_4, children [WavePlayer_44, WavePlayer_45, WavePlayer_46, WavePlayer_47], properties {weights [1,1,1,1]}}
 {id=WavePlayer_34, properties {soundWaveAssetPtr=&/Spot_Creature_Otherworld_Forest_Deer_Call_01}}
 {id=WavePlayer_35, properties {soundWaveAssetPtr=&/Spot_Creature_Otherworld_Forest_Deer_Call_03}}
 {id=WavePlayer_36, properties {soundWaveAssetPtr=&/Spot_Creature_Otherworld_Forest_Deer_Call_04}}
 {id=WavePlayer_37, properties {soundWaveAssetPtr=&/Spot_Creature_Otherworld_Forest_Deer_Call_05}}
 {id=WavePlayer_38, properties {soundWaveAssetPtr=&/Spot_Creature_Otherworld_Forest_Deer_Call_06}}
 {id=WavePlayer_39, properties {soundWaveAssetPtr=&/Spot_Creature_Otherworld_Forest_Deer_Call_07}}
 {id=WavePlayer_40, properties {soundWaveAssetPtr=&/Spot_Creature_Otherworld_Forest_Deer_Call_08}}
 {id=WavePlayer_41, properties {soundWaveAssetPtr=&/Spot_Creature_Otherworld_Forest_Deer_Call_09}}
 {id=WavePlayer_42, properties {soundWaveAssetPtr=&/Spot_Creature_Otherworld_Forest_Deer_Call_10}}
 {id=WavePlayer_43, properties {soundWaveAssetPtr=&/Spot_Creature_Otherworld_Forest_Deer_Call_11}}
 {id=WavePlayer_44, properties {soundWaveAssetPtr=&Spot_Creature_Otherworld_Forest_Deer_Call_04}}
 {id=WavePlayer_45, properties {soundWaveAssetPtr=&Spot_Creature_Otherworld_Forest_Deer_Call_05}}
 {id=WavePlayer_46, properties {soundWaveAssetPtr=&Spot_Creature_Otherworld_Forest_Deer_Call_06}}
 {id=WavePlayer_47, properties {soundWaveAssetPtr=&Spot_Creature_Otherworld_Forest_Deer_Call_09}}
]
properties {
 attenuationSettings=/Game/Sounds/Attenuation/Fort_TileAmbient_OneShot_Distant_Creature_Att
 bEnableBusSends=false
 maxDistance=8500
 soundClassObject=&Fort_Ambient_OneShots_Outdoors
 volumeMultiplier=0.2
 preEffectBusSends [{}]
}
