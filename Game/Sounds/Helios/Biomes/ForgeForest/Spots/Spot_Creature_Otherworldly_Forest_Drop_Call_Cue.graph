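# Sound cue for an otherworldly creature "drop call" spot ambience in the Forge Forest biome, played as a quiet outdoor ambient one-shot.
# All quality levels share one Modulator that varies pitch and volume over a Random node picking among fifteen equally weighted call waves.
# It uses bird one-shot attenuation and a concurrency override capping simultaneous instances.
# Distance-curved sends feed the exterior ambient source bus and a distant reverb submix.
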SoundCue /Game/Sounds/Helios/Biomes/ForgeForest/Spots/Spot_Creature_Otherworldly_Forest_Drop_Call_Cue
root=QualityLevel_0
nodes [
 {id=Modulator_0, children [Random_1], properties {pitchMax=1.1, pitchMin=0.8, volumeMax=1, volumeMin=0.8}}
 {id=QualityLevel_0, children [Modulator_0,Modulator_0,Modulator_0,Modulator_0]}
 {
  id=Random_1
  children [
   WavePlayer_0
   WavePlayer_1
   WavePlayer_2
   WavePlayer_4
   WavePlayer_7
   WavePlayer_8
   WavePlayer_9
   WavePlayer_10
   WavePlayer_11
   WavePlayer_12
   WavePlayer_13
   WavePlayer_14
   WavePlayer_15
   WavePlayer_16
   WavePlayer_17
  ]
  properties {weights [1,1,1,1,1,1,1,1,1,1,1,1,1,1,1]}
 }
 {id=WavePlayer_0, properties {soundWaveAssetPtr=&/Spot_Creature_Otherworld_Forest_Drop_Call_15}}
 {id=WavePlayer_1, properties {soundWaveAssetPtr=&/Spot_Creature_Otherworld_Forest_Drop_Call_14}}
 {id=WavePlayer_10, properties {soundWaveAssetPtr=&/Spot_Creature_Otherworld_Forest_Drop_Call_08}}
 {id=WavePlayer_11, properties {soundWaveAssetPtr=&/Spot_Creature_Otherworld_Forest_Drop_Call_07}}
 {id=WavePlayer_12, properties {soundWaveAssetPtr=&/Spot_Creature_Otherworld_Forest_Drop_Call_06}}
 {id=WavePlayer_13, properties {soundWaveAssetPtr=&/Spot_Creature_Otherworld_Forest_Drop_Call_05}}
 {id=WavePlayer_14, properties {soundWaveAssetPtr=&/Spot_Creature_Otherworld_Forest_Drop_Call_04}}
 {id=WavePlayer_15, properties {soundWaveAssetPtr=&/Spot_Creature_Otherworld_Forest_Drop_Call_03}}
 {id=WavePlayer_16, properties {soundWaveAssetPtr=&/Spot_Creature_Otherworld_Forest_Drop_Call_02}}
 {id=WavePlayer_17, properties {soundWaveAssetPtr=&/Spot_Creature_Otherworld_Forest_Drop_Call_01}}
 {id=WavePlayer_2, properties {soundWaveAssetPtr=&/Spot_Creature_Otherworld_Forest_Drop_Call_13}}
 {id=WavePlayer_4, properties {soundWaveAssetPtr=&/Spot_Creature_Otherworld_Forest_Drop_Call_12}}
 {id=WavePlayer_7, properties {soundWaveAssetPtr=&/Spot_Creature_Otherworld_Forest_Drop_Call_11}}
 {id=WavePlayer_8, properties {soundWaveAssetPtr=&/Spot_Creature_Otherworld_Forest_Drop_Call_10}}
 {id=WavePlayer_9, properties {soundWaveAssetPtr=&/Spot_Creature_Otherworld_Forest_Drop_Call_09}}
]
properties {
 attenuationSettings=&Fort_TileAmbient_OneShot_Bird_Att
 bOverrideConcurrency=true
 maxDistance=8005
 soundClassObject=&Fort_Ambient_OneShots_Outdoors
 volumeMultiplier=0.1
 busSends [{
   audioBus=null
   maxSendDistance=5000
   maxSendLevel=1.2
   minSendDistance=3000
   minSendLevel=0.25
   sendLevel=0
   soundSourceBus=&AmbientAudio_Exterior_SourceBus
   customSendLevelCurve {editorCurveData {keys [<at 0 0.25>,<at 1 1>]}}
  }]
 concurrencyOverrides {maxCount=2}
 soundSubmixSends [{
   maxSendDistance=8000
   minSendDistance=4500
   minSendLevel=0
   soundSubmix=&Fortnite_Reverb_Distant
   customSendLevelCurve {editorCurveData {keys [<at 0 0.25>,<at 1 0.55>]}}
  }]
}
