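# Sound cue for a quiet, distant otherworldly creature moaning one-shot, an outdoor ambient spot sound in the ForgeForest biome.
# A quality-level switch feeds one shared modulator into its first two levels, leaving the other two empty.
# The modulator varies a random pick among four equally weighted waves.
# It uses a shared tile-ambient attenuation asset, caps concurrency through an override, and has a distance-based send to the distant reverb submix.
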SoundCue /Game/Sounds/Helios/Biomes/ForgeForest/Spots/Spot_Creature_Otherworldly_Moaning_Distant_Cue
root=QualityLevel_0
nodes [
 {id=Modulator_0, children [Random_1]}
 {id=QualityLevel_0, children [Modulator_0,Modulator_0,null,null]}
 {id=Random_1, children [WavePlayer_7, WavePlayer_8, WavePlayer_9, WavePlayer_10], properties {weights [1,1,1,1]}}
 {id=WavePlayer_10, properties {soundWaveAssetPtr=&/Spot_Creature_Otherworldly_Moaning_Distant_04}}
 {id=WavePlayer_7, properties {soundWaveAssetPtr=&/Spot_Creature_Otherworldly_Moaning_Distant_01}}
 {id=WavePlayer_8, properties {soundWaveAssetPtr=&/Spot_Creature_Otherworldly_Moaning_Distant_02}}
 {id=WavePlayer_9, properties {soundWaveAssetPtr=&/Spot_Creature_Otherworldly_Moaning_Distant_03}}
]
properties {
 attenuationSettings=/Game/Sounds/Attenuation/Fort_TileAmbient_OneShot_Att_Closer
 bOverrideConcurrency=true
 maxDistance=8500
 soundClassObject=&Fort_Ambient_OneShots_Outdoors
 volumeMultiplier=0.05
 attenuationOverrides {bAttenuate=false}
 concurrencyOverrides {maxCount=2}
 soundSubmixSends [{
   maxSendDistance=8000
   minSendDistance=3000
   minSendLevel=0
   soundSubmix=&Fortnite_Reverb_Distant
   customSendLevelCurve {editorCurveData {keys [<at 0 0.25>,<at 0.6 0.3>,<at 1 0.66>]}}
  }]
}
