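# Sound cue for one-shot autumn leaf and tree wind rustle ambience spots in the ForgeForest biome, using outdoor one-shot sound class and attenuation.
# A quality switch sends the first two levels to a random pick between 17 regular and 17 short waves, each pool equally likely.
# The other two levels share a six-wave subset excluded from branch culling.
# Each path has its own modulator randomizing pitch and volume.
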SoundCue /Game/Sounds/Helios/Biomes/ForgeForest/Spots/Spot_Wind_Autumnal_Leaves_Tree_Rustle_Cue
root=QualityLevel_0
nodes [
 {id=Modulator_1, children [Random_4], properties {pitchMax=1.1, pitchMin=1, volumeMax=1, volumeMin=0.25}}
 {id=Modulator_2, children [Random_5], properties {pitchMax=1.1, volumeMax=1.1, volumeMin=0.1}}
 {id=QualityLevel_0, children [Modulator_1,Modulator_1,Modulator_2,Modulator_2]}
 {
  id=Random_0
  children [
   WavePlayer_0
   WavePlayer_1
   WavePlayer_2
   WavePlayer_3
   WavePlayer_4
   WavePlayer_5
   WavePlayer_6
   WavePlayer_7
   WavePlayer_8
   WavePlayer_9
   WavePlayer_10
   WavePlayer_11
   WavePlayer_12
   WavePlayer_13
   WavePlayer_14
   WavePlayer_15
   WavePlayer_16
  ]
  properties {weights [1,1,1,1,1,1,1,1,1,1,1,1,1,1,1,1,1]}
 }
 {
  id=Random_1
  children [
   WavePlayer_17
   WavePlayer_18
   WavePlayer_19
   WavePlayer_20
   WavePlayer_21
   WavePlayer_22
   WavePlayer_23
   WavePlayer_36
   WavePlayer_37
   WavePlayer_38
   WavePlayer_39
   WavePlayer_40
   WavePlayer_41
   WavePlayer_42
   WavePlayer_43
   WavePlayer_44
   WavePlayer_45
  ]
  properties {weights [1,1,1,1,1,1,1,1,1,1,1,1,1,1,1,1,1]}
 }
 {id=Random_4, children [Random_0,Random_1], properties {weights [1,1]}}
 {
  id=Random_5
  children [WavePlayer_52, WavePlayer_53, WavePlayer_54, WavePlayer_55, WavePlayer_56, WavePlayer_57]
  properties {bShouldExcludeFromBranchCulling=true, weights [1,1,1,1,1,1]}
 }
 {id=WavePlayer_0, properties {soundWaveAssetPtr=&/Spot_Wind_Autumnal_Leaves_01}}
 {id=WavePlayer_1, properties {soundWaveAssetPtr=&/Spot_Wind_Autumnal_Leaves_02}}
 {id=WavePlayer_10, properties {soundWaveAssetPtr=&/Spot_Wind_Autumnal_Leaves_11}}
 {id=WavePlayer_11, properties {soundWaveAssetPtr=&/Spot_Wind_Autumnal_Leaves_12}}
 {id=WavePlayer_12, properties {soundWaveAssetPtr=&/Spot_Wind_Autumnal_Leaves_13}}
 {id=WavePlayer_13, properties {soundWaveAssetPtr=&/Spot_Wind_Autumnal_Leaves_14}}
 {id=WavePlayer_14, properties {soundWaveAssetPtr=&/Spot_Wind_Autumnal_Leaves_15}}
 {id=WavePlayer_15, properties {soundWaveAssetPtr=&/Spot_Wind_Autumnal_Leaves_16}}
 {id=WavePlayer_16, properties {soundWaveAssetPtr=&/Spot_Wind_Autumnal_Leaves_17}}
 {id=WavePlayer_17, properties {soundWaveAssetPtr=&/Spot_Wind_Autumnal_Leaves_Short_01}}
 {id=WavePlayer_18, properties {soundWaveAssetPtr=&/Spot_Wind_Autumnal_Leaves_Short_02}}
 {id=WavePlayer_19, properties {soundWaveAssetPtr=&/Spot_Wind_Autumnal_Leaves_Short_03}}
 {id=WavePlayer_2, properties {soundWaveAssetPtr=&/Spot_Wind_Autumnal_Leaves_03}}
 {id=WavePlayer_20, properties {soundWaveAssetPtr=&/Spot_Wind_Autumnal_Leaves_Short_04}}
 {id=WavePlayer_21, properties {soundWaveAssetPtr=&/Spot_Wind_Autumnal_Leaves_Short_05}}
 {id=WavePlayer_22, properties {soundWaveAssetPtr=&/Spot_Wind_Autumnal_Leaves_Short_06}}
 {id=WavePlayer_23, properties {soundWaveAssetPtr=&/Spot_Wind_Autumnal_Leaves_Short_07}}
 {id=WavePlayer_3, properties {soundWaveAssetPtr=&/Spot_Wind_Autumnal_Leaves_04}}
 {id=WavePlayer_36, properties {soundWaveAssetPtr=&/Spot_Wind_Autumnal_Leaves_Short_08}}
 {id=WavePlayer_37, properties {soundWaveAssetPtr=&/Spot_Wind_Autumnal_Leaves_Short_09}}
 {id=WavePlayer_38, properties {soundWaveAssetPtr=&/Spot_Wind_Autumnal_Leaves_Short_10}}
 {id=WavePlayer_39, properties {soundWaveAssetPtr=&/Spot_Wind_Autumnal_Leaves_Short_11}}
 {id=WavePlayer_4, properties {soundWaveAssetPtr=&/Spot_Wind_Autumnal_Leaves_05}}
 {id=WavePlayer_40, properties {soundWaveAssetPtr=&/Spot_Wind_Autumnal_Leaves_Short_12}}
 {id=WavePlayer_41, properties {soundWaveAssetPtr=&/Spot_Wind_Autumnal_Leaves_Short_13}}
 {id=WavePlayer_42, properties {soundWaveAssetPtr=&/Spot_Wind_Autumnal_Leaves_Short_14}}
 {id=WavePlayer_43, properties {soundWaveAssetPtr=&/Spot_Wind_Autumnal_Leaves_Short_15}}
 {id=WavePlayer_44, properties {soundWaveAssetPtr=&/Spot_Wind_Autumnal_Leaves_Short_16}}
 {id=WavePlayer_45, properties {soundWaveAssetPtr=&/Spot_Wind_Autumnal_Leaves_Short_17}}
 {id=WavePlayer_5, properties {soundWaveAssetPtr=&/Spot_Wind_Autumnal_Leaves_06}}
 {id=WavePlayer_52, properties {soundWaveAssetPtr=&Spot_Wind_Autumnal_Leaves_Short_11}}
 {id=WavePlayer_53, properties {soundWaveAssetPtr=&Spot_Wind_Autumnal_Leaves_Short_13}}
 {id=WavePlayer_54, properties {soundWaveAssetPtr=&Spot_Wind_Autumnal_Leaves_Short_06}}
 {id=WavePlayer_55, properties {soundWaveAssetPtr=&Spot_Wind_Autumnal_Leaves_03}}
 {id=WavePlayer_56, properties {soundWaveAssetPtr=&Spot_Wind_Autumnal_Leaves_05}}
 {id=WavePlayer_57, properties {soundWaveAssetPtr=&/Spot_Wind_Autumnal_Leaves_08}}
 {id=WavePlayer_6, properties {soundWaveAssetPtr=&/Spot_Wind_Autumnal_Leaves_07}}
 {id=WavePlayer_7, properties {soundWaveAssetPtr=&Spot_Wind_Autumnal_Leaves_08}}
 {id=WavePlayer_8, properties {soundWaveAssetPtr=&/Spot_Wind_Autumnal_Leaves_09}}
 {id=WavePlayer_9, properties {soundWaveAssetPtr=&/Spot_Wind_Autumnal_Leaves_10}}
]
properties {
 attenuationSettings=/Game/Sounds/Attenuation/Fort_TileAmbient_OneShot_Wider_Att
 bEnableBusSends=false
 maxDistance=8000
 soundClassObject=&Fort_Ambient_OneShots_Outdoors
 volumeMultiplier=0.501
 preEffectBusSends [{}]
}
